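# Looping ambient SoundCue named for a volcano geyser, built from a reused exterior border wind loop.
# A single looping wave player feeds the output at reduced volume, in an outdoor-ducking Athena ambient sound class.
# Overridden capsule attenuation adds distance low-pass filtering, stereo spread and reverb settings over a long falloff.
# Overridden concurrency allows a few simultaneous instances and stops the quietest.
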SoundCue /Figment_S09/Athena/Sounds/Ambient/Geysers/Geyser_Loop_Volcano_02_Cue
root=WavePlayer_1
nodes [{id=WavePlayer_1, properties {bLooping=true, soundWaveAssetPtr=/FNGameplayCues/Sounds/Fort_Ambient_Looping/Wind_Exterior/border_wind_05Loop}}]
properties {
 bOverrideAttenuation=true
 bOverrideConcurrency=true
 maxDistance=15500
 soundClassObject=/Game/Sounds/FortSoundClassesAndModes/Fort_Ambient_DuckOutdoors_Athena_PSM
 volumeMultiplier=0.5
 attenuationOverrides {
  attenuationShape=Capsule
  falloffDistance=6000
  lPFFrequencyAtMax=500
  lPFRadiusMax=4000
  lPFRadiusMin=1000
  reverbWetLevelMax=20
  stereoSpread=4096
  attenuationShapeExtents=<9500 4096>
 }
 concurrencyOverrides {maxCount=3, resolutionRule=StopQuietest}
}
