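# Sound cue for a looping ambient geyser bed, built from a reused exterior border wind loop.
# An Enveloper cycles pitch and volume curves indefinitely over the wave to vary the loop.
# An Attenuation node adds capsule falloff with low-pass filtering and a curve peaking partway out, feeding a Mixer whose first input is empty.
# Cue-level overrides cap concurrency by stopping the quietest instance.
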SoundCue /Figment_S09/Athena/Sounds/Ambient/Geysers/Geyser_Loop_Z_Cue
root=Mixer_0
nodes [
 {
  id=Attenuation_0
  children [Enveloper_0]
  properties {
   bOverrideAttenuation=true
   attenuationOverrides {
    attenuationShape=Capsule
    bAttenuateWithLPF=true
    bEnableFocusInterpolation=true
    distanceAlgorithm=Custom
    falloffDistance=4000
    focusReleaseInterpSpeed=0.25
    lPFFrequencyAtMax=1000
    nonFocusVolumeAttenuation=0.25
    nonSpatializedRadiusStart=4096
    stereoSpread=2048
    attenuationShapeExtents=<5000 2048>
    customAttenuationCurve {editorCurveData {keys [<at 0 0.25>,<at 0.4 1>,<at 1 0>]}}
   }
  }
 }
 {
  id=Enveloper_0
  children [WavePlayer_1]
  properties {
   bLoopIndefinitely=true
   loopEnd=40
   pitchCurve {editorCurveData {keys [{time=0}, <at 20 0.8>, {time=40}]}}
   volumeCurve {editorCurveData {keys [<at 0 1>,<at 10 0.7>,<at 20 1>]}}
  }
 }
 {id=Mixer_0, children [null, Attenuation_0], properties {inputVolume [2,0.45]}}
 {id=WavePlayer_1, properties {bLooping=true, soundWaveAssetPtr=/FNGameplayCues/Sounds/Fort_Ambient_Looping/Wind_Exterior/border_wind_05Loop}}
]
properties {
 bOverrideConcurrency=true
 maxDistance=9000
 soundClassObject=&Fort_Mission_Ingame_Important_PSM
 volumeMultiplier=0.25
 attenuationOverrides {
  attenuationShape=Capsule
  bEnableFocusInterpolation=true
  bEnableListenerFocus=true
  falloffDistance=7500
  focusAttackInterpSpeed=10
  focusReleaseInterpSpeed=0.2
  nonFocusVolumeAttenuation=0.25
  nonSpatializedRadiusStart=4096
  attenuationShapeExtents=<9993.2539 4096>
 }
 concurrencyOverrides {maxCount=3, resolutionRule=StopQuietest}
}
